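; Parent Blueprint for the Brimstone brazier prop in the Helios environment, with toggleable flames and a point light.
; Construction and BeginPlay both run Disable/EnableFlames, which reads DisableFlames to show or hide the embers and apply a full or dimmed emissive tint to the mesh materials.
; Construction also stores the light's attenuation radius as the base value and shrinks the radius on Nintendo Switch.
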(event EventBeginPlay
  (CallFunction|Disable/EnableFlames))

(event Collision|EventActorBeginOverlap
  (params (OtherActor :type Actor%20Object%20Reference)))

(event EventTick
  (params (DeltaSeconds :type Float%20%28single-precision%29)))

(fn Disable/EnableFlames
  (bind _baseemissivevalue (Variables|Default|GetBaseEmissiveValue))
  (bind _returnvalue (Math|Conversions|ToVector(LinearColor) _baseemissivevalue))
  (if (Variables|Building|TimeOfDayLights|GetDisableFlames)
    (Rendering|SetVisibility (Variables|Default|GetEmbers) false false)
    (Rendering|Material|SetVectorParameterValueonMaterials (Variables|Debug|GetStaticMeshComponent) "SFX_Emissive_Tint(Expand)" (/ _returnvalue "6.000000,6.000000,6.000000"))
    (else
      (Rendering|SetVisibility (Variables|Default|GetEmbers) true false)
      (Rendering|Material|SetVectorParameterValueonMaterials (Variables|Debug|GetStaticMeshComponent) "SFX_Emissive_Tint(Expand)" _returnvalue))))

(fn ConstructionScript
  (bind _pointlight (Variables|Default|GetPointLight))
  (bind _pointlight_2 (Variables|Default|GetPointLight))
  (bind _attenuationradius (Variables|/Script/Engine.LocalLightComponent|GetAttenuationRadius _pointlight_2))
  (|CallParentFunction)
  (CallFunction|Disable/EnableFlames)
  (Variables|Default|SetBaseAttenuation _attenuationradius)
  (if (TouchPlatform|IsNintendoSwitch)
    (Rendering|Lighting|SetAttenuationRadius _pointlight (* _attenuationradius 0.8))
    (else
      (Rendering|Lighting|SetAttenuationRadius _pointlight _attenuationradius))))
